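# NS_MeshTrail is a Niagara mesh trail effect from the FX reuse examples, built from a single MeshTrails emitter.
# An instantaneous burst spawns two scaled-up mesh particles, drawn by one mesh renderer.
# Particles are not killed when their lifetime elapses and are set to die on emitter deactivation if infinitely lived, while ScaleColor adjusts RGB and alpha.
# No user parameters are exposed, and no module moves the particles.
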NiagaraSystem /Game/NeverCook/FX/Niagara_FX_Reuse_Examples/FXMesh/NS_MeshTrail
dependencies {usedDataInterfaces [], usedDynamicInputs [], usedModules [&SystemState]}
emitters {MeshTrails {
  emitterUpdateScript [
   &EmitterState(
    Inactive Response=::NewEnumerator1
    Life Cycle Mode=::NewEnumerator1
    Loop Behavior=::NewEnumerator1
    Loop Duration Mode=::NewEnumerator1
    UseLoopDelay=false
   )
   &SpawnBurst_Instantaneous(Spawn Count=2, Use Spawn Probability=false) {module=SpawnBurst_Instantaneous001}
  ]
  particleSpawnScript [&InitializeParticle(Mesh Scale=<100 100 100>, Mesh Scale Mode=::NewEnumerator4, UsePositionOffset=false, Write Material Random=true)]
  particleUpdateScript [
   &ParticleState(Kill Particles When Lifetime Has Elapsed=false, Let Infinitely Lived Particles Die When Emitter Deactivates=true)
   &ScaleColor(ScaleA=true, ScaleRGB=true)
  ]
  rendererClasses [&NiagaraMeshRendererProperties]
  renderers [{rendererClass=&NiagaraMeshRendererProperties, rendererIndex=0}]
 }}
summary {
 systemName=NS_MeshTrail
 emitters [{emitterName=MeshTrails, rendererClasses [&NiagaraMeshRendererProperties]}]
 userVariables []
}
