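# NS_Mesh_Basic_Color is a Niagara reuse example that shows one solid-colored mesh particle from an emitter named Egg.
# An instantaneous burst spawns a short-lived particle with a vertically stretched mesh scale and a random material value.
# Update sets a fixed muted blue color and uses a float curve to shrink the mesh to zero near end of life.
# A skeletal mesh user parameter is exposed, but none of the shown modules read it.
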NiagaraSystem /Game/NeverCook/FX/Niagara_FX_Reuse_Examples/FXMesh/NS_Mesh_Basic_Color
dependencies {usedDataInterfaces [&NiagaraDataInterfaceCurve], usedDynamicInputs [], usedModules [&SystemState]}
emitters {Egg {
  emitterUpdateScript [
   &EmitterState(Inactive Response=::NewEnumerator1, Life Cycle Mode=::NewEnumerator1, Loop Behavior=::NewEnumerator1, UseLoopDelay=false)
   &SpawnBurst_Instantaneous(Spawn Count=1, Use Spawn Probability=false)
  ]
  particleSpawnScript [&InitializeParticle(
    Color Mode=::NewEnumerator0
    Lifetime=2
    Mesh Scale=&Multiply_Vector(A=<1 1 1.5>, B=&VectorFromFloat(Value=1))
    Mesh Scale Mode=::NewEnumerator4
    UsePositionOffset=false
    Write Material Random=true
   )]
  particleUpdateScript [
   &ParticleState()
   &Color(Color=<rgba 0.0663 0.2132 0.3125 1>)
   &ScaleMeshSize(Scale Factor=&VectorFromFloat(Value=&FloatFromCurve(FloatCurve=DataInterface() {propertyValues {Curve {keys [<at 0.85 1>,<at 1 0>]}}})))
  ]
  rendererClasses [&NiagaraMeshRendererProperties]
  renderers [{rendererClass=&NiagaraMeshRendererProperties, rendererIndex=0}]
 }}
summary {
 systemName=NS_Mesh_Basic_Color
 emitters [{emitterName=Egg, rendererClasses [&NiagaraMeshRendererProperties]}]
 userVariables [User.NewNiagaraDataInterfaceSkeletalMesh {
   type=NiagaraDataInterfaceSkeletalMesh
   defaultValue {
    struct=NiagaraExt_VariableValue_DataInterface
    value {dataInterface=&self:NiagaraDataInterfaceSkeletalMesh_0}
   }
  }]
}
