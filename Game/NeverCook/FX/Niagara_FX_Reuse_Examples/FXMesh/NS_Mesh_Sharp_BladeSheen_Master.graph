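# Niagara master system, filed under reuse examples, for a sheen effect on a sharp blade, built from a single mesh-renderer emitter, Sheen_Mesh.
# It spawns particles at a steady rate, each a slightly stretched mesh with a bright HDR orange color and a randomized material value, over a randomized loop duration.
# Dynamic material parameters drive the look, with one falling from one to zero along a cubic curve and others held constant.
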NiagaraSystem /Game/NeverCook/FX/Niagara_FX_Reuse_Examples/FXMesh/NS_Mesh_Sharp_BladeSheen_Master
dependencies {usedDataInterfaces [&NiagaraDataInterfaceCurve], usedDynamicInputs [], usedModules [&SystemState]}
emitters {Sheen_Mesh {
  emitterUpdateScript [
   &EmitterState(
    Inactive Response=::NewEnumerator1
    Life Cycle Mode=::NewEnumerator1
    Loop Duration=&RandomRangeFloat(Minimum=3, Maximum=6)
    UseLoopDelay=false
   )
   &SpawnRate(SpawnRate=2, Use Spawn Probability=false)
  ]
  particleSpawnScript [&InitializeParticle(Color=<rgba 10 5.206 2.01 2>, Mesh Scale=<1.05 1 1>, Mesh Scale Mode=::NewEnumerator4, UsePositionOffset=false, Write Material Random=true)]
  particleUpdateScript [
   &ParticleState()
   &DynamicMaterialParameters(
    Index 0 Param 1=&FloatFromCurve(FloatCurve=DataInterface() {propertyValues {Curve {keys [{interpMode=RCIM_Cubic, tangentMode=RCTM_User, time=0, value=1}, {interpMode=RCIM_Cubic, tangentMode=RCTM_User, time=1, value=0}]}}})
    Index 0 Param 2=0.2
    Index 0 Param 3=-0.1
    Module.Param0WriteEnabled=true
    Module.Param1WriteEnabled=true
    Module.Param2WriteEnabled=true
    Module.Param3WriteEnabled=false
   ) {module=DynamicMaterialParameters001}
  ]
  rendererClasses [&NiagaraMeshRendererProperties]
  renderers [{rendererClass=&NiagaraMeshRendererProperties, rendererIndex=0}]
 }}
summary {
 systemName=NS_Mesh_Sharp_BladeSheen_Master
 emitters [{emitterName=Sheen_Mesh, rendererClasses [&NiagaraMeshRendererProperties]}]
 userVariables []
}
