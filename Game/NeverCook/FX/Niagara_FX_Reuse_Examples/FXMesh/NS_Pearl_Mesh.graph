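# Niagara system in the FX reuse examples that shows a single static mesh particle, named as a pearl.
# Its one emitter, ParticleMesh, burst-spawns one particle at a fixed enlarged scale with a set yaw orientation and a material random value, drawn by a mesh renderer.
# The particle outlives its lifetime and has no motion modules; the update is set to let infinitely lived particles die when the emitter deactivates.
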NiagaraSystem /Game/NeverCook/FX/Niagara_FX_Reuse_Examples/FXMesh/NS_Pearl_Mesh
dependencies {usedDataInterfaces [], usedDynamicInputs [], usedModules [&SystemState]}
emitters {ParticleMesh {
  emitterUpdateScript [
   &EmitterState(
    Inactive Response=::NewEnumerator1
    Life Cycle Mode=::NewEnumerator1
    Loop Behavior=::NewEnumerator1
    Loop Duration Mode=::NewEnumerator1
    UseLoopDelay=false
   )
   &SpawnBurst_Instantaneous(Spawn Count=1, Use Spawn Probability=false)
  ]
  particleSpawnScript [
   &InitializeParticle(Mesh Scale=<2 2 2>, Mesh Scale Mode=::NewEnumerator4, UsePositionOffset=false, Write Material Random=true)
   &InitialMeshOrientation(Mesh Orientation Mode=::NewEnumerator2, Rotation=<0 0 0.25>, Use Rotation Vector=true)
  ]
  particleUpdateScript [&ParticleState(Kill Particles When Lifetime Has Elapsed=false, Let Infinitely Lived Particles Die When Emitter Deactivates=true)]
  rendererClasses [&NiagaraMeshRendererProperties]
  renderers [{rendererClass=&NiagaraMeshRendererProperties, rendererIndex=0}]
 }}
summary {
 systemName=NS_Pearl_Mesh
 emitters [{emitterName=ParticleMesh, rendererClasses [&NiagaraMeshRendererProperties]}]
 userVariables []
}
