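# Niagara effect from the FX reuse examples that shows a ghost skull as a mesh particle, judging by its names.
# Its single SkullMeshes emitter burst-spawns scaled-down meshes with a camera offset, fixed initial rotation, overbright tint and material random value.
# Update applies drag and color scaling, and particles are not killed when their lifetime elapses.
# An FX_Alpha_CV user float is exposed, though no shown module input reads it.
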NiagaraSystem /Game/NeverCook/FX/Niagara_FX_Reuse_Examples/FXMesh/NS_SkullGhost
dependencies {usedDataInterfaces [], usedDynamicInputs [], usedModules [&SystemState]}
emitters {SkullMeshes {
  emitterUpdateScript [
   &EmitterState(Life Cycle Mode=::NewEnumerator1, Loop Behavior=::NewEnumerator1, Loop Duration Mode=::NewEnumerator1, UseLoopDelay=false)
   &SpawnBurst_Instantaneous(Use Spawn Probability=false, UseLoopCountLimit=false)
  ]
  particleSpawnScript [
   &InitializeParticle(
    Color=<rgba 1.3 1.3 1.3 2>
    Lifetime=2
    Mesh Scale=&Multiply_VectorByFloat(Vector=<1 1 1>, Float=0.3)
    Mesh Scale Mode=::NewEnumerator4
    Sprite Size Mode=::NewEnumerator3
    Uniform Sprite Size=5
    UsePositionOffset=false
    Write Material Random=true
   )
   &CameraOffset()
   &InitialMeshOrientation(Mesh Orientation Mode=::NewEnumerator2, Rotation=<0 0.25 0>, Use Rotation Vector=true)
  ]
  particleUpdateScript [
   &Drag(Use Linear Drag=true, Use Rotational Drag=true)
   &SolveForcesAndVelocity(Clamp Velocity=false, Limit Acceleration=false)
   &ParticleState(Kill Particles When Lifetime Has Elapsed=false)
   &ScaleColor(ScaleA=true, ScaleRGB=true)
  ]
  rendererClasses [&NiagaraMeshRendererProperties]
  renderers [{rendererClass=&NiagaraMeshRendererProperties, rendererIndex=0}]
 }}
summary {
 systemName=NS_SkullGhost
 emitters [{emitterName=SkullMeshes, rendererClasses [&NiagaraMeshRendererProperties]}]
 userVariables [User.FX_Alpha_CV {type=NiagaraFloat, defaultValue=1, description=""}]
}
