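# Niagara example system that, by its name, plays a vertex animation texture (VAT) rigid body simulation on a mesh particle.
# A single VAT emitter burst-spawns tinted mesh particles with a fixed lifetime matching the loop duration.
# Dynamic material parameters drive the playback: a curve-eased lerp sweeps Param 1 across what appears to be the frame range, alongside two constants.
# No user parameters are exposed.
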NiagaraSystem /Game/NeverCook/FX/Niagara_FX_Reuse_Examples/FXMesh/NS_VAT_RigidBodyDynamics
dependencies {usedDataInterfaces [&NiagaraDataInterfaceCurve], usedDynamicInputs [], usedModules [&SystemState]}
emitters {VAT {
  emitterUpdateScript [
   &EmitterState(Life Cycle Mode=::NewEnumerator1, Loop Behavior=::NewEnumerator1, Loop Duration=2, UseLoopDelay=false)
   &SpawnBurst_Instantaneous(Use Spawn Probability=false, UseLoopCountLimit=false)
  ]
  particleSpawnScript [&InitializeParticle(
    Color=<rgba 0.402 0.0212 0.0666 1>
    Lifetime=2
    Mesh Scale Mode=::NewEnumerator3
    Sprite Rotation Mode=::NewEnumerator2
    UsePositionOffset=false
    Write Material Random=true
   )]
  particleUpdateScript [
   &ParticleState()
   &DynamicMaterialParameters(
    Index 0 Param 1=&Lerp_Float(
     A=1
     B=64
     Alpha=&FloatFromCurve(FloatCurve=DataInterface() {propertyValues {Curve {keys [
          {arriveTangent=0.441, interpMode=RCIM_Cubic, leaveTangent=0.441, tangentMode=RCTM_User, time=0, value=0}
          {arriveTangent=0.9102, interpMode=RCIM_Cubic, leaveTangent=0.9102, tangentMode=RCTM_User, time=1, value=1}
         ]}}})
    )
    Index 0 Param 2=165
    Index 0 Param 3=40
    Module.Index0Param0Write=true
    Module.Index0Param1Write=true
    Module.Index0Param2Write=true
    Module.Index0Param3Write=true
   )
  ]
  rendererClasses [&NiagaraMeshRendererProperties]
  renderers [{rendererClass=&NiagaraMeshRendererProperties, rendererIndex=0}]
 }}
summary {
 systemName=NS_VAT_RigidBodyDynamics
 emitters [{emitterName=VAT, rendererClasses [&NiagaraMeshRendererProperties]}]
 userVariables []
}
